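# Sound cue that plays the equip sound for the Figment FiendHunter sniper crossbow.
# A short fixed delay feeds a single wave player that reuses a crossbow reload sample, with no randomization.
# It uses weapon reload attenuation and the weapon foley sound class, with slightly reduced volume.
# The cue overrides concurrency and has its debug flag enabled.
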SoundCue /Figment_Sniper_Crossbow_FiendHunter/Sounds/SC_Figment_Sniper_Crossbow_FiendHunter_Equip
root=Delay_0
nodes [
 {id=Delay_0, children [WavePlayer_0], properties {delayMax=0.15, delayMin=0.15}}
 {id=WavePlayer_0, properties {soundWaveAssetPtr=&/sfx_crossbow_reload_04}}
]
properties {
 attenuationSettings=&Fort_Weapon_Reload_Att
 bDebug=true
 bOverrideConcurrency=true
 maxDistance=2250
 soundClassObject=&Fort_WeaponFoley
 volumeMultiplier=0.702
}
